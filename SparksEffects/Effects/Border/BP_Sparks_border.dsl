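; BP_Sparks_border is a Blueprint that feeds the player's position to the Sparks border effect's materials.
; Every tick it gets player pawn 0 and checks that it is valid.
; It converts the pawn's location to a linear color and writes it to the PlayerLocation vector in the MC_Sparks_Border Material Parameter Collection, so materials reading that collection can react to the player's position.
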(event EventTick
  (params (DeltaSeconds :type Float%20%28single-precision%29))
  (bind _returnvalue (Game|GetPlayerPawn 0))
  (if (Utilities|IsValid _returnvalue :!class /Script/Engine.KismetSystemLibrary)
    (Rendering|Material|SetVectorParameterValue "/SparksEffects/Effects/Border/MC_Sparks_Border.MC_Sparks_Border" "PlayerLocation" (Math|Conversions|ToLinearColor(Vector) (Transformation|GetActorLocation _returnvalue)) :!class /Script/Engine.KismetMaterialLibrary)))

(fn ConstructionScript)
