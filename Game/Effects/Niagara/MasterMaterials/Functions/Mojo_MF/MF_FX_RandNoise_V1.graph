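# Material function that generates blocky pseudo-random scalar noise from texture coordinates, one value per grid cell.
# Tiled UVs are floored, multiplied by a Vector2 seed and wrapped with Frac, then scrambled by a dot product with an offset copy and a final Frac.
# A Custom HLSL node nudges even Add values so they keep a decimal part.
# The final multiply squares the R component, leaving G unused.
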MaterialFunction /Game/Effects/Niagara/MasterMaterials/Functions/Mojo_MF/MF_FX_RandNoise_V1
graph {nodes [
  Constant() {r=1024}
  n1 = FunctionInput(Preview=TextureCoordinate()) {bUsePreviewValueAsDefault=true, desc="Defaults to TexCoord[0]", inputName=In Tex Coord, inputType=FunctionInput_Vector2, sortPriority=-1}
  n2 = Multiply(
   A=@n1
   B=FunctionInput(Preview=Constant() {r=50}) {bUsePreviewValueAsDefault=true, inputName=Tex Coord Tiling, inputType=FunctionInput_Scalar}
  ) {constB=50}
  n3 = Constant2Vector() {g=456.78, r=123.45}
  n4 = Frac(None=Multiply(
    A=Floor(None=@n2)
    B=FunctionInput(Preview=@n3) {bUsePreviewValueAsDefault=true, inputName=Mult, inputType=FunctionInput_Vector2, sortPriority=1}
   ))
  MaterialExpressionFunctionInput_1 = FunctionInput(Preview=Constant() {r=45.32}) {bUsePreviewValueAsDefault=true, description=Must Have Decimal Value, inputName=Add, inputType=FunctionInput_Scalar, sortPriority=2}
  n5 = Custom!(x=@MaterialExpressionFunctionInput_1) {
   code="if(  x %2 ==0 ){x +=.001;}\r\nreturn x;"
   desc=Make Sure It Has Decimal Value
   description=Has Decimal
   outputType=CMOT_Float1
   inputs [{inputName=x, input {expression=&self:MaterialExpressionFunctionInput_1}}]
  }
  n6 = Add(A=@n4, B=@n5)
  n7 = MaterialFunctionCall(Float2=Add(A=@n4, B=DotProduct(A=@n4, B=@n6))) {materialFunction=&BreakOutFloat2Components}
  FunctionOutput(None=Frac(None=Multiply(A=@n7.R, B=@n7.R)))
 ]}
