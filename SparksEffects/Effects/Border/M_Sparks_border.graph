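# Translucent, unlit, two-sided material for a Sparks border effect, revealed around the player position read from MC_Sparks_Border.
# A player-centered sphere mask gates opacity shaped by fresnel, UV edge and depth fades, with optional screen-space hologram scanlines.
# Flickering world-aligned noise feeds the scanlines and a vertex-offset branch, while emissive blends far, medium and close colors.
# Named reroutes organize ripple, grid and distorted world-UV branches.
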Material /SparksEffects/Effects/Border/M_Sparks_border
graph {
 properties {blendMode=BLEND_Translucent, pixelDepthOffsetMode=PDOM_Legacy, shadingModel=MSM_Unlit, twoSided=true, usedShadingModels=Unlit}
 nodes [
  n1 = ActorPositionWS()
  n2 = MaterialFunctionCall(Float2=Round(None=Divide(A=@n1) {constB=2000})) {materialFunction=&BreakOutFloat2Components}
  n3 = Add(A=Time(), B=Add(A=@n2.R, B=@n2.R))
  n4 = Multiply(A=ScalarParameter() {defaultValue=0.5, group=Hologram, parameterName=FlickerSpeed}, B=@n3)
  n5 = Sine(None=@n4) {period=1.6}
  n6 = Sine(None=@n4) {period=1.42}
  n7 = Ceil(None=AppendVector(A=AppendVector(A=@n5, B=@n6), B=Sine(None=@n4) {period=1.51}))
  Add(A=@n7) {constB=0.1}
  n8 = Constant() {r=1}
  n9 = Multiply(A=Time(), B=Constant() {desc=Ripple Speed Mult, r=0.35})
  n10 = LinearInterpolate(Alpha=PerInstanceRandom()) {constB=100}
  n11 = WorldPosition()
  n12 = NamedRerouteUsage()
  n13 = Constant() {desc=smaller radius, r=128}
  n14 = ScalarParameter() {defaultValue=0.1, parameterName=Falloff}
  n15 = NamedRerouteDeclaration(None=Multiply(A=SphereMask(A=@n11.XYZ, B=@n12, Radius=@n13, Hardness=@n14)) {constB=0.5}) {name=SmallSphereMask, nodeColor=<rgba 0 0.4844 1 1>}
  n16 = Frac(None=Add(A=Add(A=@n9, B=@n10), B=@n15) {desc=Add ripple Offset Map})
  n17 = Multiply(A=Add(A=Subtract(A=@n16, B=@n8), B=@n15), B=Constant() {desc=ripple contrast mult, r=25})
  Clamp(None=@n17, Max=Multiply(A=Constant() {desc=ripple count clamp, r=15}) {constB=2}) {maxDefault=50}
  n18 = Abs(None=Constant() {r=64})
  ComponentMask(None=Divide(A=WorldPosition() {worldPositionShaderOffset=WPT_ExcludeAllShaderOffsets}.XYZ, B=Multiply(A=@n18) {constB=-1})) {g=true, r=true}
  Constant() {r=128}
  n19 = ScalarParameter() {defaultValue=2, desc="Random Tiling ", parameterName=GlobalTIling}
  Divide(A=@n19)
  n20 = Constant3Vector() {constant=<rgba 64 64 64 0>}
  n21 = Divide(A=WorldPosition() {worldPositionShaderOffset=WPT_ExcludeAllShaderOffsets}.XYZ, B=Multiply(A=Abs(None=@n20)) {constB=-1})
  n22 = VertexNormalWS()
  n23 = Abs(None=ComponentMask(None=@n22) {r=true})
  n24 = Constant() {desc=projection contrast scale, r=1}
  n25 = LinearInterpolate(
   A=ComponentMask(None=@n21) {b=true, r=true}
   B=ComponentMask(None=@n21) {b=true, g=true}
   Alpha=ComponentMask(None=MaterialFunctionCall(In=@n23, Contrast=@n24) {materialFunction=&CheapContrast}.Result) {r=true}
  ) {desc=xy uv}
  n26 = Abs(None=ComponentMask(None=@n22) {b=true})
  LinearInterpolate(
   A=@n25
   B=ComponentMask(None=@n21) {g=true, r=true}
   Alpha=ComponentMask(None=MaterialFunctionCall(In=@n26, Contrast=@n24) {materialFunction=&CheapContrast}.Result) {r=true}
  ) {desc=xyz uv}
  n27 = NamedRerouteUsage()
  n28 = LinearInterpolate(
   A=@n27
   B=MaterialFunctionCall!(
    In Texture
    Random Color Texture
    In Texcoords=@n27
    Turn Time=Constant() {r=0.275}
    Distortion Tiling=Constant() {r=1.75}
    Distortion Amount=Constant() {desc=distortion amount, r=5}
    Tiling
    Detail Factor
   ) {materialFunction=/Game/Effects/Niagara/MasterMaterials/Functions/Mojo_MF/MF_FX_TurnDistortion}.Out Base UV
   Alpha=NamedRerouteUsage()
  )
  n29 = Multiply(A=@n28, B=@n19)
  n30 = NamedRerouteDeclaration(None=Frac(None=@n29)) {name=UVFrac, nodeColor=<rgba 1 0.9844 0 1>}
  n31 = Multiply(A=@n28, B=Multiply(A=@n19) {constB=0.5})
  n32 = MaterialFunctionCall(Float2=Sine(None=Frac(None=@n31))) {materialFunction=&BreakOutFloat2Components}
  n33 = Multiply(A=Multiply(A=@n32.R, B=@n32.R)) {constB=0.5}
  LinearInterpolate(A=@n30, B=NamedRerouteDeclaration(None=Abs(None=@n33)) {name=VignietteSoftEdge, nodeColor=<rgba 0.125 1 0 1>}) {constAlpha=0.8, desc=debug}
  n34 = Constant() {r=5}
  n35 = Divide(A=Floor(None=Multiply(A=@n27, B=@n19)), B=@n19)
  LinearInterpolate(Alpha=Saturate(None=MaterialFunctionCall!(In Tex Coord=@n35, Tex Coord Tiling, Mult, Add=@n34) {materialFunction=/Game/Effects/Niagara/MasterMaterials/Functions/Mojo_MF/MF_FX_RandNoise_V1}.Result)) {constA=0.001, constB=0.005}
  n36 = NamedRerouteUsage()
  LinearInterpolate(
   A=@n28
   B=MaterialFunctionCall!(UV=@n28, Bulge Intensity XY=LinearInterpolate(Alpha=@n36) {constB=5}) {materialFunction=/Game/Effects/Niagara/MasterMaterials/Functions/Mojo_MF/MF_FX_Bulge_UV}.Out UV
   Alpha=@n36
  )
  MaterialFunctionCall!(
   Angle=MaterialFunctionCall!(Degrees=Constant() {desc=UV Rotation Angle Global, r=75}) {materialFunction=/Engine/Functions/Engine_MaterialFunctions01/Math/DegreesToRadians}.Radians
   UV/2DVector=@n27
  ) {materialFunction=/Game/Characters/Player/Male/Medium/Bodies/M_MED_Faux/Materials/MF_2DRotate}
  MaterialFunctionCall() {materialFunction=&BoundingBoxBased_0-1_UVW}
  MaterialFunctionCall() {materialFunction=&ObjectScale}
  MaterialFunctionCall!(UVs, Size, SizeScale) {materialFunction=/Game/Packages/Fortress_SharedMaterials/Base_Material_Functions/SquarePatternFromUVs}
  MaterialFunctionCall!(Center Point World Position) {materialFunction=/Engine/Functions/Engine_MaterialFunctions02/UVs/WorldSpaceAlignedScreenCoordinates}
  MaterialFunctionCall!(Texture Coordinate) {materialFunction=/Engine/Functions/Engine_MaterialFunctions02/WorldPositionOffset/UVLayoutToWorldSpacePosition}
  MaterialFunctionCall(Float2=TextureCoordinate()) {desc=checking directions of each uv gradient, materialFunction=&BreakOutFloat2Components}
  MaterialFunctionCall() {materialFunction=&WorldAlignedTexture}
  MaterialFunctionCall() {materialFunction=&WorldAlignedTexture}
  n37 = Time()
  n38 = Saturate(None=Sine(None=Multiply(A=@n37) {constB=0.65}))
  n39 = Sine(None=MaterialFunctionCall(Multiplier=@n17) {materialFunction=&Pi}.Result) {period=6.28}
  n40 = Add(A=Multiply(A=Constant() {desc=Fade Weight Mult, r=1}, B=Constant() {r=0.8}), B=Constant() {r=0.2})
  n41 = Multiply(A=Saturate(None=Subtract(A=@n40, B=@n16)), B=@n15) {desc=mask}
  n42 = Saturate(None=Multiply(A=Multiply(A=@n41, B=@n39)) {constB=0.65})
  Multiply(A=@n42, B=LinearInterpolate(Alpha=@n38) {constA=0.1, constB=0.55})
  n43 = Constant() {r=350}
  Multiply(A=@n42, B=OneMinus(None=DepthFade(FadeDistance=@n43))) {desc=Ripples when close up?}
  n44 = Multiply(A=Add(A=@n42, B=@n15) {desc=limit ripples to small sphere mask gradient}) {constB=5, desc=boost white point a bit overall}
  Multiply(A=@n44, B=LinearInterpolate(Alpha=@n38) {constA=0.5})
  Multiply(A=@n27, B=@n19)
  n45 = NamedRerouteUsage()
  Multiply(A=@n45, B=Constant() {r=4}) {constB=4}
  NamedRerouteDeclaration(None=Saturate(None=@n44)) {name=Ripple Mask, nodeColor=<rgba 1 0 0.5625 1>}
  NamedRerouteDeclaration(None=@n25) {name=World Space UVs, nodeColor=<rgba 0 0.5547 1 1>}
  n46 = NamedRerouteUsage()
  n47 = Normalize(VectorInput=Subtract(A=WorldPosition().XYZ, B=ObjectPositionWS()))
  n48 = Constant3Vector() {constant=<rgba 1 1 0 0>}
  n49 = ScalarParameter() {defaultValue=10, group=Noise, parameterName=WPOStrength}
  n50 = Multiply(A=ObjectRadius(), B=ScalarParameter() {defaultValue=0.025, group=Noise, parameterName=WPOStrengthFromObjectRadius}) {constB=0.05}
  n51 = TextureObject() {samplerType=SAMPLERTYPE_LinearColor, texture=&T_NoisePack_LinearColor}
  n52 = Divide(A=ObjectRadius(), B=ScalarParameter() {defaultValue=0.1, group=Noise, parameterName=NoiseSize})
  n53 = Multiply(A=VectorParameter() {group=Noise, parameterName=NoiseTileMult, defaultValue=<rgba 1 1 0.1 1>}.RGB, B=@n52)
  n54 = Time()
  n55 = VectorParameter() {group=Noise, parameterName=NoisePanSpeed, defaultValue=<rgba -0.08 0.07 -0.2 0>}
  n56 = Add(A=WorldPosition().XYZ, B=Multiply(A=@n52, B=Multiply(A=@n54, B=@n55.RGB)))
  n57 = Multiply(
   A=Power(Base=MaterialFunctionCall(TextureObject=@n51, TextureSize=@n53, WorldPosition=@n56) {materialFunction=&WorldAlignedTexture}.XY Texture) {constExponent=4}
   B=ScalarParameter() {defaultValue=2, group=Noise, parameterName=NoiseMaskMult}
  ) {constB=20}
  n58 = ScalarParameter() {defaultValue=0.5, group=Noise, parameterName=NoiseFrequency}
  n59 = Add(A=Sine(None=@n4) {period=0.72}, B=Multiply(B=@n58) {constA=0.75}) {constB=0.75}
  n60 = Multiply(A=Add(A=Sine(None=@n4) {period=2.1}, B=Multiply(B=@n58) {constA=0.25}), B=@n59)
  n61 = Multiply(B=@n58) {constA=0.9, constB=0.9}
  n62 = Sine(None=@n4) {period=1.25}
  n63 = Multiply(A=@n59, B=Ceil(None=Add(A=@n61, B=@n62) {constB=0.9}))
  n64 = Sine(None=@n3) {period=0.075}
  n65 = Multiply(A=Saturate(None=Multiply(A=@n60, B=@n63)), B=LinearInterpolate(Alpha=ConstantBiasScale(None=@n64)) {constA=0.7})
  n66 = Add(
   A=ScalarParameter() {group=Noise, parameterName=NoiseAdd}
   B=Multiply(A=Multiply(A=@n7, B=@n65), B=ScalarParameter() {defaultValue=2, group=Noise, parameterName=NoiseMult})
  )
  n67 = Desaturation(None=Saturate(None=Multiply(A=@n57, B=@n66)))
  n68 = Multiply(A=@n67, B=Multiply(A=Multiply(A=@n47, B=@n48), B=Add(A=@n49, B=@n50))) {constB=0}
  NamedRerouteDeclaration(None=StaticSwitchParameter(True=Multiply(A=@n68, B=@n46), False=Constant()) {defaultValue=true, group=Hologram, parameterName=Is Hologram, sortPriority=1}) {name=WPO, nodeColor=<rgba 1 0 0.0703 1>}
  NamedRerouteDeclaration(None=Divide(A=Floor(None=@n29), B=@n19)) {name=RandUVTile, nodeColor=<rgba 0 1 0.2031 1>}
  n69 = ScalarParameter() {defaultValue=0.015, parameterName=Grid Line Thickness}
  NamedRerouteDeclaration(None=Multiply(A=MaterialFunctionCall!(In TexCoords=@n45, Width=@n69) {materialFunction=/Game/Effects/Niagara/MasterMaterials/Functions/Mojo_MF/MF_FX_SquareEdgeMask}.Mask)) {name=gridtest, nodeColor=<rgba 1 0 0.8906 1>}
  n70 = ScalarParameter() {defaultValue=0.015, parameterName=Grid Line Thickness}
  n71 = Multiply(A=Max(A=MaterialFunctionCall(In TexCoords=@n30, Width=@n70) {materialFunction=&MF_FX_SquareEdgeMask}.Mask) {constB=0.1})
  NamedRerouteDeclaration(None=LinearInterpolate(Alpha=Saturate(None=@n71)) {constA=0.3, constB=0.6}) {name=SparseGridPattern, nodeColor=<rgba 0 1 0.9531 1>}
  n72 = ConstantBiasScale(None=ScreenPosition().ViewportUV) {bias=-0.5, scale=2}
  n73 = MaterialFunctionCall(Float2=ViewSize()) {materialFunction=&BreakOutFloat2Components}
  n74 = AppendVector(A=Constant() {r=1}, B=Divide(A=@n73.G, B=@n73.G))
  n75 = Multiply(A=ConstantBiasScale(None=Multiply(A=@n72, B=@n74) {desc=Keep 1:1 aspect ratio regardless of screen aspect ratio}) {scale=2}) {constB=0.5}
  NamedRerouteDeclaration(None=@n75) {name=ScreenSpaceUV, nodeColor=<rgba 1 0.9141 0 1>}
  NamedRerouteUsage()
  NamedRerouteUsage()
  NamedRerouteUsage()
  NamedRerouteUsage()
  NamedRerouteUsage()
  NamedRerouteUsage()
  OneMinus(None=@n16) {desc=mask}
  n76 = VectorParameter() {parameterName=Close Color, defaultValue=<rgba 1 0.0417 0.0417 1>}
  n77 = LinearInterpolate(
   A=MaterialFunctionCall(
    A=VectorParameter() {parameterName=Far Color, defaultValue=<rgba 0 0.8482 1 1>}.RGB
    B=VectorParameter() {parameterName=Med Color, defaultValue=<rgba 1 0.3833 0.0312 1>}.RGB
    C=@n76.RGB
    Alpha=NamedRerouteUsage()
   ) {materialFunction=&3ColorBlend}.Result
   B=@n76.RGB
   Alpha=Saturate(None=NamedRerouteUsage())
  )
  n78 = ScalarParameter() {defaultValue=10, parameterName=Emissive Intensity}
  Reroute(None=Multiply(
    A=Multiply(A=@n77, B=@n78)
    B=MaterialFunctionCall() {materialFunction=&LinearGradient}.VGradient
   )) {desc=Out emissive color}
  n79 = Fresnel(ExponentIn=ScalarParameter() {defaultValue=0.5, parameterName=Fresnel Opacity Exp})
  n80 = Multiply(
   A=MaterialFunctionCall() {materialFunction=&LinearGradient}.VGradient
   B=Power(Base=OneMinus(None=@n79), Exp=ScalarParameter() {defaultValue=1, parameterName=Fresnel Power Exp})
  )
  n81 = QualitySwitch(
   Default=MaterialFunctionCall(In=@n67, Contrast=ScalarParameter() {defaultValue=0.2, group=Noise, parameterName=NoiseContrast}) {materialFunction=&CheapContrast}.Result
   Low=Constant()
  )
  n82 = Saturate(None=@n81)
  n83 = Multiply(A=ScalarParameter() {defaultValue=60, parameterName=ScanlineAmount}, B=LinearInterpolate(Alpha=Ceil(None=@n82)) {constA=1, constB=1.25})
  n84 = ScreenPosition()
  n85 = MaterialFunctionCall(Float2=ViewSize()) {materialFunction=&BreakOutFloat2Components}
  n86 = AppendVector(A=Constant() {r=1}, B=Divide(A=@n85.G, B=@n85.G))
  n87 = ConstantBiasScale(None=Multiply(A=ConstantBiasScale(None=@n84.ViewportUV) {bias=-0.5, scale=2}, B=@n86) {desc=Keep 1:1 aspect ratio regardless of screen aspect ratio}) {scale=2}
  n88 = Constant2Vector() {g=-1, r=1}
  n89 = MaterialFunctionCall(World Position=ActorPositionWS() {desc=Also use Object Position}) {materialFunction=&TransformToClipSpace}
  n90 = Multiply(A=Multiply(A=@n86, B=Multiply(A=@n88, B=@n89.Clip Space XY))) {constB=0.5, desc=Halve offsets to account for 0...1 UV space}
  n91 = Add(
   A=MaterialFunctionCall(UVs=Subtract(A=@n87, B=@n90), "Rotation Angle (0-1)"=Constant() {r=0.25}) {materialFunction=&CustomRotator}.Rotated Values
   B=Multiply(A=Time(), B=Constant() {r=-0.02})
  )
  n92 = ConstantBiasScale(None=Sine(None=Multiply(A=@n83, B=@n91) {constA=1}))
  n93 = Add(A=Constant() {r=0.5}, B=@n92)
  n94 = ScalarParameter() {defaultValue=-0.4, group=Noise, parameterName=NoiseBrightness}
  n95 = ComponentMask(None=Subtract(A=WorldPosition().XYZ, B=ObjectPositionWS())) {b=true}
  n96 = ObjectRadius()
  n97 = Saturate(None=ConstantBiasScale(None=Divide(A=@n95, B=@n96)))
  n98 = LinearInterpolate(
   A=Multiply(A=Multiply(A=@n94, B=@n82) {constB=0.1}, B=Saturate(None=@n93))
   B=LinearInterpolate(A=@n97, B=OneMinus(None=@n97), Alpha=ScalarParameter() {group=Hologram, parameterName=FlipGradient})
   Alpha=ScalarParameter() {defaultValue=0.4, group=Hologram, parameterName=VerticalGradientInflience}
  )
  n99 = Multiply(A=@n81) {constB=1.5}
  n100 = Subtract(A=ScalarParameter() {defaultValue=0.6, group=Hologram, parameterName=ScanlineMinOpacity}, B=@n99)
  n101 = Power(Base=@n92) {constExponent=1.5}
  n102 = Multiply(
   A=DepthFade(Opacity=Clamp(None=@n98), FadeDistance=ScalarParameter() {defaultValue=5, group=Hologram, parameterName=FadeDistance})
   B=LinearInterpolate(
    A=Clamp(None=@n100) {minDefault=-0.25}
    B=Add(A=@n99)
    Alpha=MaterialFunctionCall(Float2=@n101) {materialFunction=&BreakOutFloat2Components}.R
   ) {constA=0.5}
  )
  n103 = StaticSwitchParameter(
   True=Multiply(A=Saturate(None=@n102), B=ScalarParameter() {defaultValue=5, group=Hologram, parameterName=OpacityMult})
   False=NamedRerouteUsage()
  ) {parameterName=Use Scanline Pattern?}
  n104 = WorldPosition()
  n105 = NamedRerouteDeclaration(None=ComponentMask(None=CollectionParameter() {collection=&/MC_Sparks_Border, parameterName=PlayerLocation}) {b=true, g=true, r=true}) {name=playerLocation, nodeColor=<rgba 0 1 1 1>}
  n106 = ScalarParameter() {defaultValue=320, parameterName=Radius}
  n107 = ScalarParameter() {defaultValue=0.1, parameterName=Falloff}
  n108 = LinearInterpolate(
   B=StaticSwitchParameter(True=Multiply(A=@n103, B=@n80), False=@n80) {defaultValue=true, group=Hologram, parameterName=Is Hologram, sortPriority=1}
   Alpha=NamedRerouteDeclaration(None=SphereMask(A=@n104.XYZ, B=@n105, Radius=@n106, Hardness=@n107)) {name=Sphere Mask Shape, nodeColor=<rgba 0.0781 0 1 1>}
  )
  n109 = MaterialFunctionCall!(
   TexCoords=MaterialFunctionCall(UVs=TextureCoordinate(), Texture Scale=Constant2Vector() {g=0.95, r=0.95}) {materialFunction=&ScaleUVsByCenter}.UVs
   Edge Sharpness=Constant() {r=4}
   Edge Subtract=Constant() {r=0.1}
  ) {materialFunction=/Game/Effects/Niagara/MasterMaterials/Functions/Mojo_MF/MF_FX_UV_EdgeMask}
  n110 = Saturate(None=Multiply(A=Power(Base=@n109.1-Edge Mask) {constExponent=4}) {constB=2})
  n111 = ScalarParameter() {defaultValue=250, parameterName=Global Depth Fade Distance}
  n112 = Multiply(
   A=Multiply(A=Multiply(A=@n108, B=@n110), B=DepthFade(FadeDistance=@n111)) {desc=For ground and other intersections}
   B=ScalarParameter() {defaultValue=0.15, parameterName=Global Opacity Mult}
  )
  n113 = OneMinus(None=Fresnel())
  Saturate(None=Multiply(
    A=Multiply(A=@n112, B=@n113)
    B=MaterialFunctionCall() {materialFunction=&CameraDepthFade}.Result
   ))
  ScalarParameter() {defaultValue=320, parameterName=Radius}
  n114 = TransformPosition(None=WorldPosition().XYZ) {transformSourceType=TRANSFORMPOSSOURCE_World}
  n115 = TransformPosition(None=ObjectPositionWS()) {transformSourceType=TRANSFORMPOSSOURCE_World}
  n116 = MaterialFunctionCall() {materialFunction=&ObjectLocalBounds}
  n117 = Add(A=Divide(A=Subtract(A=@n114, B=@n115), B=@n116.Local Bounds Size)) {constB=0.5}
  n118 = MaterialFunctionCall(Float3=@n117) {materialFunction=&BreakOutFloat3Components}
  n119 = MaterialFunctionCall(Float3=@n116.Local Bounds Size) {materialFunction=&BreakOutFloat3Components}
  n120 = Constant()
  n121 = Saturate(None=If(A=@n119.R, B=@n119.R, "A > B"=@n120, "A == B"=@n120, "A < B"=Constant() {r=1}))
  n122 = MaterialFunctionCall(Float3=@n117) {materialFunction=&BreakOutFloat3Components}
  StaticSwitchParameter(
   True=AppendVector(A=LinearInterpolate(A=@n118.R, B=@n118.R, Alpha=@n121) {desc=pick largest axis based on object scale}, B=OneMinus(None=@n118.B)) {desc="tex coords local object(x or y, Z) option"}
   False=AppendVector(A=@n122.R, B=@n122.R) {desc="tex coords local object(x , y) option"}
  ) {parameterName="Local UV Direction Flip (True is for Upright Meshes Texcord V => Mesh Zup)"}
  n123 = Multiply(
   A=Constant2Vector() {g=-1, r=1}
   B=MaterialFunctionCall(World Position=ActorPositionWS() {desc=Also use Object Position}) {materialFunction=&TransformToClipSpace}.Clip Space XY
  )
  Subtract(A=@n75, B=Multiply(A=Multiply(A=@n74, B=@n123)) {constB=0.5, desc=Halve offsets to account for 0...1 UV space})
  TextureCoordinate()
 ]
}
